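# Animation Blueprint for the TaxiHeel large male player head cosmetic that flags eyelid movement to the head's materials.
# On each animation update it calls the parent, then packs the left and right blink pose curves into a 2D vector.
# It compares that vector with the inherited previous blink pose within a threshold and writes the result as the HasLidChanged scalar on the part's skeletal mesh materials.
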AnimBlueprint /BRCosmetics/Characters/Player/Male/Large/Heads/M_LRG_TaxiHeel_Head/Meshes/M_LRG_TaxiHeel_Head_AnimBP
target_skeleton=/BRCosmetics/Characters/Player/Male/Large/Heads/M_LRG_TaxiHeel_Head/Meshes/CustomSkeleton/Fortnite_Base_Head_TaxiHeel_Export_Skeleton
graphs [EventGraph {nodes [
   n1 = Event BlueprintUpdateAnimation()
   > Parent: BlueprintUpdateAnimation!(execute, DeltaTimeX:float=@n1.DeltaTimeX) -> (then) {is=title K2Node_CallParentFunction}
   n2 = MakeVector2D(X=GetCurveValue(CurveName=L_blink_pose), Y=GetCurveValue(CurveName=R_blink_pose))
   n3 = Get Prev Blink Pose() -> (Prev Blink Pose:Vector 2D Structure)
   n4 = Get Blink Pose Threshold() -> (Blink Pose Threshold:float)
   > SetScalarParameterValueOnMaterials!(
    execute
    self:Mesh Component Object Reference=GetPartSkeletalMeshComponent()
    ParameterName:Name=HasLidChanged
    ParameterValue:float="To Float (Boolean)"(InBool="Not Equal (Vector2D)"!(A:Vector 2D Structure=@n2, B:Vector 2D Structure=@n3, ErrorTolerance:float=@n4) -> (ReturnValue:bool=~) {is=title K2Node_PromotableOperator})
   ) -> (then)
   TryGetPawnOwner()
  ]}]
variables []
